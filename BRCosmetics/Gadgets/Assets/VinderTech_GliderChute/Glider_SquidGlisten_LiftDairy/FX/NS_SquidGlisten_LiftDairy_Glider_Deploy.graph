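# Niagara deploy effect for the SquidGlisten LiftDairy glider, a brief sprite flash.
# Corona and Star_Beams burst immediately, with time-of-day scaling on their fade curves, while Horizontal_Waves bursts after a short delay.
# Beams and waves spread rotation evenly by normalized particle count and stretch through non-uniform size curves.
# The corona randomizes a dynamic material parameter, grows over life, and scales its size by User.FrontendScale in the front end.
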NiagaraSystem /BRCosmetics/Gadgets/Assets/VinderTech_GliderChute/Glider_SquidGlisten_LiftDairy/FX/NS_SquidGlisten_LiftDairy_Glider_Deploy
dependencies {
 usedDataInterfaces [&NiagaraDataInterfaceColorCurve, &NiagaraDataInterfaceCurve, &NiagaraDataInterfaceVector2DCurve]
 usedDynamicInputs [&Multiply_Float]
 usedModules [&SystemState, &FN_Integrator]
}
emitters {
 Corona {
  emitterUpdateScript [
   &EmitterState()
   &SpawnBurst_Instantaneous(Spawn Count=1, Use Spawn Probability=false)
  ]
  particleSpawnScript [
   &InitializeParticle(Lifetime=0.7, Sprite Size Mode=::NewEnumerator3, Uniform Sprite Size=125, UsePositionOffset=true, Write Material Random=true)
   &DynamicMaterialParameters(
    Index 0 Param 1=&RandomRangeFloat(Minimum=0.4, Maximum=0.6)
    Index 0 Param 2=1
    Index 0 Param 3=1
    Index 0 Param 4=1
    Module.Param0WriteEnabled=true
    Module.Param1WriteEnabled=true
    Module.Param2WriteEnabled=true
    Module.Param3WriteEnabled=true
   )
  ]
  particleUpdateScript [
   &ParticleState()
   &ScaleColor(
    Scale Mode=::NewEnumerator1
    Scale RGBA=&MakeVector4_FromFloat(Float=&FloatFromCurve(
      FloatCurve=DataInterface() {propertyValues {Curve {keys [
          <at 0 0>
          {arriveTangent=-3.7291, interpMode=RCIM_Cubic, leaveTangent=-3.7291, tangentMode=RCTM_User, time=0.05, value=1}
          {arriveTangentWeight=1.0122, interpMode=RCIM_Cubic, tangentMode=RCTM_User, time=1, value=0}
         ]}}}
      Scale Curve=&FNTimeOfDayScalar(Night Scale=0.4, Morning Scale=1, Evening Scale=1)
     ))
    ScaleRGBA=true
   )
   &ScaleSpriteSize(Scale Factor=&Vector2DFromFloat(Value=&FloatFromCurve(
      FloatCurve=DataInterface() {propertyValues {Curve {keys [
          {arriveTangent=2.4947, interpMode=RCIM_Cubic, leaveTangent=2.4947, tangentMode=RCTM_User, time=0, value=0.333}
          {arriveTangent=0.503, interpMode=RCIM_Cubic, leaveTangent=0.503, tangentMode=RCTM_User, time=0.4, value=1.25}
          {tangentMode=RCTM_User, time=1, value=1.333}
         ]}}}
      Scale Curve=&FN_FrontEndFloatModifier(Gameplay=1, Front End=$User.FrontendScale)
     )))
   &CameraOffset(Camera Offset Amount=15)
  ]
  rendererClasses [&NiagaraSpriteRendererProperties]
  renderers [{rendererClass=&NiagaraSpriteRendererProperties}]
 }
 Horizontal_Waves {
  emitterUpdateScript [
   &EmitterState()
   &SpawnBurst_Instantaneous(Spawn Count=2, Spawn Time=0.725, Use Spawn Probability=false, UseLoopCountLimit=false)
  ]
  particleSpawnScript [&InitializeParticle(
    Color=<rgba 0.2 0.4 0.3667 0.4>
    Lifetime=0.4
    Sprite Size=<111 275>
    Sprite Size Mode=::NewEnumerator4
    UsePositionOffset=false
    Write Material Random=true
   )]
  particleUpdateScript [
   &ParticleState()
   &ScaleColor(
    Scale Mode=::NewEnumerator1
    Scale RGBA=&MakeVector4_FromFloat(Float=&FloatFromCurve(FloatCurve=DataInterface() {propertyValues {Curve {keys [<at 0 0>, {arriveTangent=-3.0726, interpMode=RCIM_Cubic, leaveTangent=-3.0726, tangentMode=RCTM_User, time=0.05, value=1}, <at 1 0>]}}}))
    ScaleRGBA=true
   )
   /SetVariables(Particles.SpriteRotation:NiagaraFloat=&Multiply_FloatByInt(Float=&Add_Float(A=/Game/Effects/Niagara/DynamicInputs/Math/FN_NormalizedParticleCount(), B=0.25), Integer=360)) {isSetParameters=true, module=SetVariables_#1}
   &ScaleSpriteSize(
    Non-Uniform Curve Sprite Scale=DataInterface() {propertyValues {
      XCurve {keys [
        {arriveTangent=6.2067, interpMode=RCIM_Cubic, leaveTangent=6.2067, tangentMode=RCTM_User, time=0, value=0}
        {tangentMode=RCTM_User, time=1, value=6}
       ]}
      YCurve {keys [
        {arriveTangent=2.8894, interpMode=RCIM_Cubic, leaveTangent=2.8894, tangentMode=RCTM_User, time=0, value=0}
        {tangentMode=RCTM_User, time=1, value=1.5}
       ]}
     }}
    Scale Sprite Size Mode=::NewEnumerator3
   )
   &CameraOffset(Camera Offset Amount=55)
   &SpriteFacingAndAlignment(Sprite Alignment=<0 0 1>, Sprite Facing=<0 -1 0.1>, WriteAlignment=true, WriteFacing=true) {enabled=false}
  ]
  rendererClasses [&NiagaraSpriteRendererProperties]
  renderers [{rendererClass=&NiagaraSpriteRendererProperties}]
 }
 Star_Beams {
  emitterUpdateScript [
   &EmitterState()
   &SpawnBurst_Instantaneous(Spawn Count=5, Use Spawn Probability=false, UseLoopCountLimit=false)
  ]
  particleSpawnScript [&InitializeParticle(
    Color=<rgba 0.75 1 0.9583 1>
    Lifetime=0.75
    Position Offset=<0 7 0>
    Sprite Size Mode=::NewEnumerator3
    Uniform Sprite Size=75
    UsePositionOffset=true
    Write Material Random=true
   )]
  particleUpdateScript [
   &ParticleState()
   &ScaleColor(
    Scale Mode=::NewEnumerator1
    Scale RGBA=&MakeVector4_FromFloat(Float=&FloatFromCurve(
      FloatCurve=DataInterface() {propertyValues {Curve {keys [<at 0 0>, {arriveTangent=-3.0726, interpMode=RCIM_Cubic, leaveTangent=-3.0726, tangentMode=RCTM_User, time=0.1, value=1}, <at 1 0>]}}}
      Scale Curve=&FNTimeOfDayScalar(Night Scale=0.3, Morning Scale=0.65, Day Scale=0.65, Evening Scale=0.65)
     ))
    ScaleRGBA=true
   )
   /SetVariables(Particles.SpriteRotation:NiagaraFloat=&Multiply_FloatByInt(Float=&FN_NormalizedParticleCount(), Integer=360)) {isSetParameters=true, module=SetVariables_#1}
   &ScaleSpriteSize(
    Non-Uniform Curve Sprite Scale=DataInterface() {propertyValues {
      XCurve {keys [{tangentMode=RCTM_User, time=0, value=3}, {tangentMode=RCTM_User, time=1, value=7}]}
      YCurve {keys [{tangentMode=RCTM_User, time=0, value=1}, {tangentMode=RCTM_User, time=1, value=7}]}
     }}
    Scale Sprite Size Mode=::NewEnumerator3
   )
   &SpriteFacingAndAlignment(Sprite Alignment=<0 0 1>, Sprite Facing=<0 -1 0.1>, WriteAlignment=true, WriteFacing=true) {enabled=false}
  ]
  rendererClasses [&NiagaraSpriteRendererProperties]
  renderers [{rendererClass=&NiagaraSpriteRendererProperties}]
 }
}
summary {
 systemName=NS_SquidGlisten_LiftDairy_Glider_Deploy
 emitters [
  {emitterName=Star_Beams, rendererClasses [&NiagaraSpriteRendererProperties]}
  {emitterName=Horizontal_Waves, rendererClasses [&NiagaraSpriteRendererProperties]}
  {emitterName=Corona, rendererClasses [&NiagaraSpriteRendererProperties]}
 ]
 userVariables [
  User.bIsFrontEnd {type=NiagaraBool}
  User.bIsFrontEndPreview {type=NiagaraBool}
  User.bIsFullyDeployed {type=NiagaraBool}
  User.ForwardDot {type=NiagaraFloat, defaultValue=0, description=""}
  User.FrontendScale {type=NiagaraFloat, defaultValue=0.6, description=""}
  User.RibbonIDOffset {type=NiagaraInt32}
  User.RightDot {type=NiagaraFloat, defaultValue=0, description=""}
  User.SkeletalMesh {
   type=NiagaraDataInterfaceSkeletalMesh
   defaultValue {
    struct=NiagaraExt_VariableValue_DataInterface
    value {dataInterface=&self:NiagaraDataInterfaceSkeletalMesh_0}
   }
  }
 ]
}
